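# Physics asset for the Deamon Edge pickaxe skeletal mesh, a two-handed melee cosmetic with front and back dangling tassels.
# One long capsule covers the handle, and small identical capsules sit on each tassel bone.
# Each tassel is a three-link constraint chain hung from the handle, with moderate swing limits and tight twist limits so it can sway without spinning.
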PhysicsAsset /BRCosmetics/Weapons/FORT_Melee/Pickaxe_Deamon_Edge/Meshes/SK_Pickaxe_Deamon_Edge_Physics
bodies {
 Melee_TwoHanded_Handle [{shapeType=Capsule, length=82.1732, radius=5.751, center=<0 -1.2118 8.384>, rotation=<rot 0 0 0.0001>}]
 dyn_tassel_back_01 [{shapeType=Capsule, length=1, radius=1, center=<1 0 0>}]
 dyn_tassel_back_02 [{shapeType=Capsule, length=1, radius=1, center=<1 0 0>}]
 dyn_tassel_back_03 [{shapeType=Capsule, length=1, radius=1, center=<1 0 0>}]
 dyn_tassel_front_01 [{shapeType=Capsule, length=1, radius=1, center=<1 0 0>}]
 dyn_tassel_front_02 [{shapeType=Capsule, length=1, radius=1, center=<1 0 0>}]
 dyn_tassel_front_03 [{shapeType=Capsule, length=1, radius=1, center=<1 0 0>}]
}
constraints [
 {bone1Name=dyn_tassel_back_01, bone2Name=Melee_TwoHanded_Handle, swing1LimitDegrees=45, swing2LimitDegrees=45, twistLimitDegrees=5}
 {bone1Name=dyn_tassel_back_02, bone2Name=dyn_tassel_back_01, swing1LimitDegrees=45, swing2LimitDegrees=45, twistLimitDegrees=5}
 {bone1Name=dyn_tassel_back_03, bone2Name=dyn_tassel_back_02, swing1LimitDegrees=45, swing2LimitDegrees=45, twistLimitDegrees=5}
 {bone1Name=dyn_tassel_front_01, bone2Name=Melee_TwoHanded_Handle, swing1LimitDegrees=45, swing2LimitDegrees=45, twistLimitDegrees=5}
 {bone1Name=dyn_tassel_front_02, bone2Name=dyn_tassel_front_01, swing1LimitDegrees=45, swing2LimitDegrees=45, twistLimitDegrees=5}
 {bone1Name=dyn_tassel_front_03, bone2Name=dyn_tassel_front_02, swing1LimitDegrees=45, swing2LimitDegrees=45, twistLimitDegrees=5}
]
