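# Animation Blueprint for the Deamon Edge pickaxe that layers physics motion over the Item_Pickaxe_Demon_Idle animation.
# The idle sequence feeds a RigidBody node using the pickaxe's override physics asset, custom gravity and reduced component velocity and acceleration influence, then goes to the output pose.
# A chain of Anim Dynamics nodes for front and back tassel bones, most at zero alpha, is disconnected from the output.
# The event graph holds no logic.
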AnimBlueprint /BRCosmetics/Weapons/FORT_Melee/Pickaxe_Deamon_Edge/Meshes/SK_Pickaxe_Deamon_Edge_AnimBP
target_skeleton=&/SK_Pickaxe_Deamon_Edge_Skeleton
graphs [
 AnimGraph {nodes [
   n1 = Anim Dynamics - Bone: dyn_tassel_back_01(Alpha=1, ExternalForce="0,0,0") {
    angularDampingOverride=0.98
    bGravityOverrideInSimSpace=true
    bOverrideAngularDamping=true
    bOverrideLinearDamping=true
    lODThreshold=1
    linearDampingOverride=0.98
    numSolverIterationsPostUpdate=4
    numSolverIterationsPreUpdate=16
    alphaBoolBlend {blendOption=Linear}
    boundBone {boneName=dyn_tassel_back_01}
    chainEnd {boneName=dyn_tassel_back_03}
    componentLinearAccScale=<1 1 1>
    componentLinearVelScale=<1 1 1>
    physicsBodyDefinitions [{
      sphereCollisionRadius=0
      boundBone {boneName=dyn_tassel_back_01}
      boxExtents=<5 5 5>
      constraintSetup {angularLimitsMax=<89 89 89>, angularLimitsMin=<-89 -89 -89>, angularTarget=<0 0 0>}
      localJointOffset=<0 0 0.5>
     }]
   }
   n2 = Anim Dynamics - Bone: dyn_tassel_front_01(ComponentPose=@n1, Alpha=1, ExternalForce="0,0,0") {
    angularDampingOverride=0.98
    bGravityOverrideInSimSpace=true
    bOverrideAngularDamping=true
    bOverrideLinearDamping=true
    lODThreshold=1
    linearDampingOverride=0.98
    numSolverIterationsPostUpdate=4
    numSolverIterationsPreUpdate=16
    alphaBoolBlend {blendOption=Linear}
    boundBone {boneName=dyn_tassel_front_01}
    chainEnd {boneName=dyn_tassel_front_03}
    componentLinearAccScale=<1 1 1>
    componentLinearVelScale=<1 1 1>
    physicsBodyDefinitions [{
      sphereCollisionRadius=0
      boundBone {boneName=dyn_tassel_front_01}
      boxExtents=<5 5 5>
      constraintSetup {angularLimitsMax=<89 89 89>, angularLimitsMin=<-89 -89 -89>, angularTarget=<0 0 0>}
      localJointOffset=<0 0 0.5>
     }]
   }
   n3 = Anim Dynamics - Bone: dyn_tassel_back_02(ComponentPose=@n2, Alpha=0, ExternalForce="0,0,0") {
    angularDampingOverride=0.98
    bGravityOverrideInSimSpace=true
    bOverrideAngularDamping=true
    bOverrideLinearDamping=true
    gravityScale=0.5
    lODThreshold=1
    linearDampingOverride=0.98
    numSolverIterationsPostUpdate=4
    numSolverIterationsPreUpdate=16
    alphaBoolBlend {blendOption=Linear}
    boundBone {boneName=dyn_tassel_back_02}
    chainEnd {boneName=dyn_tassel_back_03}
    componentLinearAccScale=<1 1 1>
    componentLinearVelScale=<1 1 1>
    physicsBodyDefinitions [{
      sphereCollisionRadius=0
      boundBone {boneName=dyn_tassel_back_02}
      boxExtents=<5 5 5>
      constraintSetup {angularLimitsMax=<89 89 89>, angularLimitsMin=<-89 -89 -89>, angularTarget=<0 0 0>}
      localJointOffset=<0 0 0.5>
     }]
   }
   n4 = Anim Dynamics - Bone: dyn_tassel_front_02(ComponentPose=@n3, Alpha=0, ExternalForce="0,0,0") {
    angularDampingOverride=0.98
    bGravityOverrideInSimSpace=true
    bOverrideAngularDamping=true
    bOverrideLinearDamping=true
    gravityScale=0.5
    lODThreshold=1
    linearDampingOverride=0.98
    numSolverIterationsPostUpdate=4
    numSolverIterationsPreUpdate=16
    alphaBoolBlend {blendOption=Linear}
    boundBone {boneName=dyn_tassel_front_02}
    chainEnd {boneName=dyn_tassel_front_03}
    componentLinearAccScale=<1 1 1>
    componentLinearVelScale=<1 1 1>
    physicsBodyDefinitions [{
      sphereCollisionRadius=0
      boundBone {boneName=dyn_tassel_front_02}
      boxExtents=<5 5 5>
      constraintSetup {angularLimitsMax=<89 89 89>, angularLimitsMin=<-89 -89 -89>, angularTarget=<0 0 0>}
      localJointOffset=<0 0 0.5>
     }]
   }
   n5 = Anim Dynamics - Bone: dyn_tassel_back_03(ComponentPose=@n4, Alpha=0, ExternalForce="0,0,0") {
    angularDampingOverride=0.98
    bGravityOverrideInSimSpace=true
    bOverrideAngularDamping=true
    bOverrideLinearDamping=true
    gravityScale=0.25
    lODThreshold=1
    linearDampingOverride=0.98
    numSolverIterationsPostUpdate=4
    numSolverIterationsPreUpdate=16
    alphaBoolBlend {blendOption=Linear}
    boundBone {boneName=dyn_tassel_back_03}
    chainEnd {boneName=dyn_tassel_back_03}
    componentLinearAccScale=<1 1 1>
    componentLinearVelScale=<1 1 1>
    physicsBodyDefinitions [{
      sphereCollisionRadius=0
      boundBone {boneName=dyn_tassel_back_03}
      boxExtents=<5 5 5>
      constraintSetup {angularLimitsMax=<89 89 89>, angularLimitsMin=<-89 -89 -89>, angularTarget=<0 0 0>}
      localJointOffset=<0 0 0.5>
     }]
   }
   Anim Dynamics - Bone: dyn_tassel_front_03(ComponentPose=@n5, Alpha=0, ExternalForce="0,0,0") {
    angularDampingOverride=0.98
    bGravityOverrideInSimSpace=true
    bOverrideAngularDamping=true
    bOverrideLinearDamping=true
    gravityScale=0.25
    lODThreshold=1
    linearDampingOverride=0.98
    numSolverIterationsPostUpdate=4
    numSolverIterationsPreUpdate=16
    alphaBoolBlend {blendOption=Linear}
    boundBone {boneName=dyn_tassel_front_03}
    chainEnd {boneName=dyn_tassel_front_03}
    componentLinearAccScale=<1 1 1>
    componentLinearVelScale=<1 1 1>
    physicsBodyDefinitions [{
      sphereCollisionRadius=0
      boundBone {boneName=dyn_tassel_front_03}
      boxExtents=<5 5 5>
      constraintSetup {angularLimitsMax=<89 89 89>, angularLimitsMin=<-89 -89 -89>, angularTarget=<0 0 0>}
      localJointOffset=<0 0 0.5>
     }]
   }
   n6 = RigidBody(ComponentPose=Local To Component(LocalPose=Sequence Player 'Item_Pickaxe_Demon_Idle'() {sequence=/BRCosmetics/Animation/Game/MainPlayer/Combat/Melee/TwoHanded/WideGrip/Item_Pickaxe_Demon_Idle})) {
    bOverrideWorldGravity=true
    overridePhysicsAsset=&/SK_Pickaxe_Deamon_Edge_Physics
    alphaBoolBlend {blendOption=Linear}
    componentLinearAccScale=<0.25 0.25 0.25>
    componentLinearVelScale=<0.25 0.25 0.25>
    overrideWorldGravity {z=-980}
   }
   Output Pose(Result=Component To Local(ComponentPose=@n6)) {layerGroup=None}
  ]}
 EventGraph {nodes [Event BlueprintUpdateAnimation()]}
]
variables []
